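# Sound cue for the drop sound of the Figment heavy assault weapon.
# A Mixer plays the weapon's own drop wave layered with a reused Flintlock revolver drop wave, both at equal input volume.
# The cue lowers overall volume and uses weapon drop attenuation, the loot collect sound class and a shared weapon drop concurrency set.
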SoundCue /Figment_Assault_Heavy_Athena/Audio/AssaultHeavy_Drop_Figment_Cue
root=Mixer_0
nodes [
 {id=Mixer_0, children [WavePlayer_1,WavePlayer_2], properties {inputVolume [1,1]}}
 {id=WavePlayer_1, properties {soundWaveAssetPtr=/Figment_Assault_Heavy_Athena/Audio/Wavs/AssaultHeavy_Drop_01}}
 {id=WavePlayer_2, properties {soundWaveAssetPtr=/Game/Sounds/Fort_Weapons/FlintLock/Flintlock_Revolver_Drop_01}}
]
properties {
 attenuationSettings=&Fort_Weapon_Drop_Att
 maxDistance=3000
 soundClassObject=&Fort_Loot_Collect
 volumeMultiplier=0.3604
 concurrencyOverrides {resolutionRule=PreventNew}
 concurrencySet [&Weapon_Drop_StopOldest]
}
