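# First-person firing sound cue for the Figment heavy assault rifle, reusing the base AssaultHeavy close fire waves.
# A Random node, excluded from branch culling, picks one of four equally weighted close fire variations, then a Modulator adds slight volume variation.
# The cue plays pitched up at reduced volume with first-person weapon attenuation, the ranged fire 1P sound class and a first-person gunfire concurrency limit.
# Its debug flag is switched on.
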SoundCue /Figment_Assault_Heavy_Athena/Audio/AssaultHeavy_Fire_Figment_1PCue
root=Modulator_4
nodes [
 {id=Modulator_4, children [Random_3], properties {volumeMax=1.1, volumeMin=1}}
 {
  id=Random_3
  children [WavePlayer_13, WavePlayer_14, WavePlayer_18, WavePlayer_19]
  properties {bShouldExcludeFromBranchCulling=true, weights [1,1,1,1]}
 }
 {id=WavePlayer_13, properties {soundWaveAssetPtr=/Game/Sounds/Fort_Weapons/AutoRifle/AssaultHeavy/AssaultHeavy_Fire_Close_01}}
 {id=WavePlayer_14, properties {soundWaveAssetPtr=/Game/Sounds/Fort_Weapons/AutoRifle/AssaultHeavy/AssaultHeavy_Fire_Close_02}}
 {id=WavePlayer_18, properties {soundWaveAssetPtr=/Game/Sounds/Fort_Weapons/AutoRifle/AssaultHeavy/AssaultHeavy_Fire_Close_03}}
 {id=WavePlayer_19, properties {soundWaveAssetPtr=/Game/Sounds/Fort_Weapons/AutoRifle/AssaultHeavy/AssaultHeavy_Fire_Close_04}}
]
properties {
 attenuationSettings=&Fort_Weapon_Projectile_FirstPerson_Athena_Att
 bDebug=true
 maxDistance=2097152
 pitchMultiplier=1.1
 soundClassObject=&Fort_WeaponRanged_Fire_1P_PSM
 volumeMultiplier=0.752
 concurrencySet [&Weapon_Gunfire_FirstPerson_Conc]
}
